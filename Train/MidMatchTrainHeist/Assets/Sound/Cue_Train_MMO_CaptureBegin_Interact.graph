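# Sound cue for the interaction sound when a capture begins in the mid-match train heist, in the Fort_Gadgets sound class.
# A mixer blends two equally weighted wave layers, one with close attenuation and one with distant attenuation.
# Only the close layer passes through a modulator, with pitch fixed at normal.
# The cue has boosted volume, owner-limited concurrency that stops the oldest instance, and its debug flag left on.
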SoundCue /Train/MidMatchTrainHeist/Assets/Sound/Cue_Train_MMO_CaptureBegin_Interact
root=Mixer_1
nodes [
 {
  id=Attenuation_3
  children [Modulator_2]
  properties {attenuationSettings=/Train/MidMatchTrainHeist/Assets/Sound/Settings/Train_MMOQuest_Close_Att}
 }
 {
  id=Attenuation_4
  children [WavePlayer_1]
  properties {attenuationSettings=/Train/MidMatchTrainHeist/Assets/Sound/Settings/Train_MMOQuest_Distant_Att}
 }
 {id=Mixer_1, children [Attenuation_3, Attenuation_4], properties {inputVolume [1,1]}}
 {id=Modulator_2, children [WavePlayer_0], properties {pitchMax=1, pitchMin=1}}
 {id=WavePlayer_0, properties {soundWaveAssetPtr=&/Train_MMO_CaptureBegin_Interact}}
 {id=WavePlayer_1, properties {soundWaveAssetPtr=&Train_MMO_CaptureBegin_Interact}}
]
properties {
 bDebug=true
 maxDistance=51000
 soundClassObject=&Fort_Gadgets
 volumeMultiplier=2.5
 attenuationOverrides {bSpatialize=false}
 concurrencyOverrides {bLimitToOwner=true, maxCount=3, resolutionRule=StopOldest}
}
