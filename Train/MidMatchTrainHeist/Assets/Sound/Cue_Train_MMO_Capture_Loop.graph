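# Looping sound cue for a capture in the mid-match train heist, routed through the gadgets sound class.
# A looping wave player feeds a Modulator that fixes a lowered pitch with no random variation.
# A root Enveloper then holds volume flat and ramps pitch up over a long curve, so the loop climbs as it plays.
# The cue plays at reduced volume and references the train quest loot attenuation asset.
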SoundCue /Train/MidMatchTrainHeist/Assets/Sound/Cue_Train_MMO_Capture_Loop
root=Enveloper_1
nodes [
 {
  id=Enveloper_1
  children [Modulator_0]
  properties {pitchCurve {editorCurveData {keys [<at 0 0.95>, {time=0}, <at 50 1.5>]}}, volumeCurve {editorCurveData {keys [<at 0 1>,<at 1 1>]}}}
 }
 {id=Modulator_0, children [WavePlayer_0], properties {pitchMax=0.7, pitchMin=0.7, volumeMax=1, volumeMin=1}}
 {id=WavePlayer_0, properties {bLooping=true, soundWaveAssetPtr=&/Train_MMO_Capture_Loop}}
]
properties {
 attenuationSettings=/Train/MidMatchTrainHeist/Assets/Sound/Settings/Train_MMOQuest_Loot_Att
 bDebug=true
 maxDistance=5600
 soundClassObject=&Fort_Gadgets
 volumeMultiplier=0.3
 attenuationOverrides {bSpatialize=false}
 concurrencyOverrides {bLimitToOwner=true, maxCount=3, resolutionRule=StopOldest}
}
