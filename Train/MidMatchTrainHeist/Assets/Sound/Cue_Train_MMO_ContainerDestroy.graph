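# Sound Cue for a container being destroyed in the Train mid-match heist content, using a loot-collect sound class and the train quest loot attenuation.
# A mixer layers two branches: one of three random vending-machine metal impacts with randomly lowered pitch, and a piñata explosion.
# The explosion passes through a modulator and an enveloper whose volume curve swells from near silence, then fades to zero.
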SoundCue /Train/MidMatchTrainHeist/Assets/Sound/Cue_Train_MMO_ContainerDestroy
root=Mixer_0
nodes [
 {
  id=Enveloper_0
  children [Modulator_4]
  properties {pitchCurve {editorCurveData {keys [{time=0}]}}, volumeCurve {editorCurveData {keys [<at -0.0037 0.01>,<at 0.3089 0.6784>,<at 1 0>]}}}
 }
 {id=Mixer_0, children [Modulator_5,Enveloper_0], properties {inputVolume [1,1]}}
 {id=Modulator_4, children [WavePlayer_14]}
 {id=Modulator_5, children [Random_4], properties {pitchMax=0.95, pitchMin=0.85}}
 {id=Random_4, children [WavePlayer_15, WavePlayer_16, WavePlayer_17], properties {weights [1,1,1]}}
 {id=WavePlayer_14, properties {soundWaveAssetPtr=/Game/Sounds/Fort_Human_VO/Pinata/Pinata_New_Explode}}
 {
  id=WavePlayer_15
  properties {soundWaveAssetPtr=/Game/Sounds/Fort_Impact_Sounds/Demolition/VendingMachine/Metal_VendMachine_ImpactFinal_mix_01}
 }
 {
  id=WavePlayer_16
  properties {soundWaveAssetPtr=/Game/Sounds/Fort_Impact_Sounds/Demolition/VendingMachine/Metal_VendMachine_ImpactFinal_mix_02}
 }
 {
  id=WavePlayer_17
  properties {soundWaveAssetPtr=/Game/Sounds/Fort_Impact_Sounds/Demolition/VendingMachine/Metal_VendMachine_ImpactFinal_mix_03}
 }
]
properties {
 attenuationSettings=/Train/MidMatchTrainHeist/Assets/Sound/Settings/Train_MMOQuest_Loot_Att
 bDebug=true
 maxDistance=5600
 soundClassObject=&Fort_Loot_Collect
 attenuationOverrides {bSpatialize=false}
 concurrencyOverrides {bLimitToOwner=true, maxCount=3, resolutionRule=StopOldest}
}
