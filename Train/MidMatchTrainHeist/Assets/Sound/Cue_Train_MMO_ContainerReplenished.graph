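# Sound cue for the mid-match train heist container replenished event, playing a synth reset sound in the Fort_Gadgets sound class.
# Each of two branches mixes close and distant layers through the train whistle close and distant attenuation settings.
# A random node picks between these equally weighted, like-named branches, then a modulator adds slight pitch variation.
# The cue has its debug flag enabled.
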SoundCue /Train/MidMatchTrainHeist/Assets/Sound/Cue_Train_MMO_ContainerReplenished
root=Modulator_0
nodes [
 {id=Attenuation_0, children [WavePlayer_12], properties {attenuationSettings=/Train/Sound/Settings/Train_Whistle_Close_Att}}
 {id=Attenuation_1, children [WavePlayer_13], properties {attenuationSettings=/Train/Sound/Settings/Train_Whistle_Distant_Att}}
 {id=Attenuation_3, children [WavePlayer_10], properties {attenuationSettings=&Train_Whistle_Close_Att}}
 {id=Attenuation_4, children [WavePlayer_11], properties {attenuationSettings=&Train_Whistle_Distant_Att}}
 {id=Mixer_3, children [Attenuation_3, Attenuation_4], properties {inputVolume [1,1]}}
 {id=Mixer_4, children [Attenuation_0, Attenuation_1], properties {inputVolume [1,1]}}
 {id=Modulator_0, children [Random_2], properties {pitchMax=1.025, pitchMin=0.98}}
 {id=Random_2, children [Mixer_3,Mixer_4], properties {weights [1,1]}}
 {id=WavePlayer_10, properties {soundWaveAssetPtr=&/Train_MMO_Reset_Synth_Close}}
 {id=WavePlayer_11, properties {soundWaveAssetPtr=&/Train_MMO_Reset_Synth_Distant}}
 {id=WavePlayer_12, properties {soundWaveAssetPtr=&Train_MMO_Reset_Synth_Close}}
 {id=WavePlayer_13, properties {soundWaveAssetPtr=&Train_MMO_Reset_Synth_Distant}}
]
properties {
 bDebug=true
 maxDistance=51000
 soundClassObject=&Fort_Gadgets
 volumeMultiplier=1
 attenuationOverrides {bSpatialize=false}
 concurrencyOverrides {bLimitToOwner=true, maxCount=3, resolutionRule=StopOldest}
}
