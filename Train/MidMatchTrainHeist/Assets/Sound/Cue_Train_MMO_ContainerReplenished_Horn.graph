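# Sound cue for the train horn of the mid-match train heist's container replenished event, using Japan reset-horn recordings.
# Close and distant waves play together, each through its own train whistle attenuation node, so the horn's character follows listener distance.
# A mixer blends both layers into a fixed-volume, unvaried-pitch modulator that boosts the output.
# The cue uses the gadgets sound class and has its debug flag on.
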SoundCue /Train/MidMatchTrainHeist/Assets/Sound/Cue_Train_MMO_ContainerReplenished_Horn
root=Modulator_0
nodes [
 {id=Attenuation_7, children [WavePlayer_1], properties {attenuationSettings=/Train/Sound/Settings/Train_Whistle_Close_Att}}
 {id=Attenuation_8, children [WavePlayer_6], properties {attenuationSettings=/Train/Sound/Settings/Train_Whistle_Distant_Att}}
 {id=Mixer_1, children [Attenuation_7, Attenuation_8], properties {inputVolume [1,1]}}
 {id=Modulator_0, children [Mixer_1], properties {pitchMax=1, pitchMin=1, volumeMax=1.5, volumeMin=1.5}}
 {id=WavePlayer_1, properties {soundWaveAssetPtr=/Train/MidMatchTrainHeist/Assets/Sound/Japan/Train_MMO_ResetHorn_Japan_Close}}
 {id=WavePlayer_6, properties {soundWaveAssetPtr=/Train/MidMatchTrainHeist/Assets/Sound/Japan/Train_MMO_ResetHorn_Japan_Distant}}
]
properties {
 bDebug=true
 maxDistance=51000
 pitchMultiplier=0.999
 soundClassObject=&Fort_Gadgets
 volumeMultiplier=2
 attenuationOverrides {bSpatialize=false}
 concurrencyOverrides {bLimitToOwner=true, maxCount=3, resolutionRule=StopOldest}
}
