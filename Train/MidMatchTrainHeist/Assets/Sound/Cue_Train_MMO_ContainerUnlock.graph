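# Sound cue for the container unlock in the mid-match train heist, in the gadgets sound class.
# A switch on the PlayerAffliation parameter sends its first three inputs to one mixer and the fourth to a second.
# Each mixer layers close and distant "captured" waves through separate capture-point attenuation settings, with both layers boosted.
# Both branches use identically named waves and the same attenuation and gain.
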SoundCue /Train/MidMatchTrainHeist/Assets/Sound/Cue_Train_MMO_ContainerUnlock
root=Switch_1
nodes [
 {
  id=Attenuation_0
  children [WavePlayer_7]
  properties {attenuationSettings=/Train/MidMatchTrainHeist/Assets/Sound/Settings/Train_MMOQuest_Capture_Point_Att}
 }
 {
  id=Attenuation_1
  children [WavePlayer_9]
  properties {attenuationSettings=&Train_MMOQuest_Capture_Point_Att}
 }
 {
  id=Attenuation_2
  children [WavePlayer_6]
  properties {attenuationSettings=/Train/MidMatchTrainHeist/Assets/Sound/Settings/Train_MMOQuest_Capture_Point_Distant_Att}
 }
 {
  id=Attenuation_5
  children [WavePlayer_8]
  properties {attenuationSettings=&Train_MMOQuest_Capture_Point_Distant_Att}
 }
 {id=Mixer_0, children [Attenuation_1, Attenuation_5], properties {inputVolume [1.3,1.3]}}
 {id=Mixer_2, children [Attenuation_0, Attenuation_2], properties {inputVolume [1.3,1.3]}}
 {id=Switch_1, children [Mixer_2,Mixer_2,Mixer_2,Mixer_0], properties {intParameterName=PlayerAffliation}}
 {id=WavePlayer_6, properties {soundWaveAssetPtr=&/Train_MMO_Captured_Distant}}
 {id=WavePlayer_7, properties {soundWaveAssetPtr=&/Train_MMO_Captured_Close}}
 {id=WavePlayer_8, properties {soundWaveAssetPtr=&Train_MMO_Captured_Distant}}
 {id=WavePlayer_9, properties {soundWaveAssetPtr=&Train_MMO_Captured_Close}}
]
properties {
 bDebug=true
 maxDistance=27500
 soundClassObject=&Fort_Gadgets
 volumeMultiplier=1
 attenuationOverrides {bSpatialize=false}
 concurrencyOverrides {bLimitToOwner=true, maxCount=3, resolutionRule=StopOldest}
}
