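# Sound cue for the graffiti event in the mid-match train heist, judging by its name and path.
# A mixer layers a smoke grenade explosion, played immediately, with a point-captured paint sound held back by a fixed half-second delay.
# It plays under the gadgets sound class and borrows the train quest loot attenuation settings for distance behavior.
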SoundCue /Train/MidMatchTrainHeist/Assets/Sound/Cue_Train_MMO_GrafittiEvent
root=Mixer_0
nodes [
 {id=Delay_0, children [WavePlayer_0], properties {delayMax=0.5, delayMin=0.5}}
 {id=Mixer_0, children [Delay_0,WavePlayer_4], properties {inputVolume [1,1]}}
 {id=WavePlayer_0, properties {soundWaveAssetPtr=&/Train_MMOQuest_PointCaptured_Paint}}
 {id=WavePlayer_4, properties {soundWaveAssetPtr=/Game/Athena/Sounds/Weapons/Smoke_Grenade/sfx_SmokeGrenade_Explode_02}}
]
properties {
 attenuationSettings=/Train/MidMatchTrainHeist/Assets/Sound/Settings/Train_MMOQuest_Loot_Att
 bDebug=true
 maxDistance=5600
 soundClassObject=&Fort_Gadgets
 volumeMultiplier=1
 attenuationOverrides {bSpatialize=false}
 concurrencyOverrides {bLimitToOwner=true, maxCount=3, resolutionRule=StopOldest}
}
