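# Sound cue for the hacking sound in the mid-match train heist, playing the Train_MMO_Hack_Start wave.
# A single wave player feeds a Modulator root on the gadgets sound class, with slightly reduced volume.
# It references the train quest loot attenuation asset.
# Inline non-spatialized attenuation and per-owner stop-oldest concurrency values are stored, but no flags enabling those overrides are set, and debug is left on.
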SoundCue /Train/MidMatchTrainHeist/Assets/Sound/Cue_Train_MMO_Hacking
root=Modulator_0
nodes [
 {id=Modulator_0, children [WavePlayer_1]}
 {id=WavePlayer_1, properties {soundWaveAssetPtr=&/Train_MMO_Hack_Start}}
]
properties {
 attenuationSettings=/Train/MidMatchTrainHeist/Assets/Sound/Settings/Train_MMOQuest_Loot_Att
 bDebug=true
 maxDistance=5600
 soundClassObject=&Fort_Gadgets
 volumeMultiplier=0.9
 attenuationOverrides {bSpatialize=false}
 concurrencyOverrides {bLimitToOwner=true, maxCount=3, resolutionRule=StopOldest}
}
